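# Niagara system for the Candy Corn vehicle boost cosmetic, trailing sprites and ribbons from sockets on the vehicle's skeletal mesh.
# Bats and notes01 emit animated sprites and Left001 draws orange ribbons, all spread across filtered socket counts.
# Confetti's SpawnRate is disabled, as are all painted-variant overrides.
# User.bIsFrontendPreview switches spawn rates, velocities and lifetimes, and toggles between two collision setups, one using analytical planes for preview.
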NiagaraSystem /VehicleCosmetics/Boosts/Boost_CandyCorn/FX/NS_Boost_CandyCorn
dependencies {
 usedDataInterfaces [
  &NiagaraDataInterfaceCurve
  &NiagaraDataInterfaceSpriteRendererInfo
  &NiagaraDataInterfaceSkeletalMesh
 ]
 usedDynamicInputs []
 usedModules [
  &self:SystemSpawnScript.SystemScriptSource.SystemScriptGraph.NiagaraNodeAssignment_0.SetVariables_#1
  &SystemState
 ]
}
emitters {
 Bats {
  emitterUpdateScript [
   &EmitterState(Life Cycle Mode=::NewEnumerator1, Loop Duration=2, UseLoopDelay=false)
   &SpawnRate(Use Spawn Probability=false) {enabled=false, module=SpawnRate001}
   &VC_FilteredSocketCount()
   &SpawnRate(
    SpawnRate=&MakeCustomFloatFromBool(Bool=$User.bIsFrontendPreview, True Float=15, False Float=20)
    Use Spawn Probability=false
   )
  ]
  particleSpawnScript [
   &VC_PaintedVariantsOverride() {enabled=false}
   &InitializeParticle(
    Color Mode=::NewEnumerator0
    Sprite Rotation Angle=&RandomRangeFloat(Minimum=-0.1, Maximum=0.1)
    Sprite Rotation Mode=::NewEnumerator1
    Sprite Size Max=<20 20>
    Sprite Size Min=<12 18>
    Sprite Size Mode=::NewEnumerator2
    UsePositionOffset=false
    Write Material Random=true
   )
   &PartitionParticles(Number Of Partitions=$Emitter.PortCount)
   &ShapeLocation(MaskSpawnGroup=false, Override Seed=false, Sphere Radius=50, UseOwnerScale=true)
   &SkeletalMeshLocation(
    Bone / Socket Index=$Particles.PartitionParticles.Partition
    Bone / Socket Sampling Mode=::NewEnumerator5
    MaskSpawnGroup=false
    Mesh Sampling Type=::NewEnumerator6
    Sampled Position Offset=<-20 0 0>
    Skeletal Mesh=$System.SkeletalMesh
   )
   &AddVelocity(Velocity=&MakeCustomVectorFromBool(
     Bool=$User.bIsFrontendPreview
     True Vector=&RandomRangeVector(Minimum=<-500 85 50>, Maximum=<-800 -85 200>)
     False Vector=&RandomRangeVector(Minimum=<-700 75 105>, Maximum=<-700 -75 150>)
    ))
  ]
  particleUpdateScript [
   &ParticleState()
   &ScaleSpriteSize(Scale Factor=<6 6>, Scale Sprite Size Mode=::NewEnumerator2)
   &SubUVAnimation(
    "Sprite Renderer "=DataInterface() {propertyValues {SpriteRenderer=&self:Bats_0.NiagaraSpriteRendererProperties_0}}
    SubUV Animation Mode:ENiagara_SubUVLookupModeV2=::NewEnumerator1
    SubUV Curve=DataInterface() {propertyValues {Curve {keys [{arriveTangent=4.1785, time=0, value=0}, <at 0.3 0.55>, {arriveTangent=0.2086, leaveTangent=0.2086, time=1, value=1}]}}}
    UseEndFrame=false
    UseStartFrame=false
   ) {enabled=false}
   &SubUVAnimation(
    Random Start Frame=true
    "Sprite Renderer "=DataInterface() {propertyValues {SpriteRenderer=&self:Bats_0.NiagaraSpriteRendererProperties_0}}
    SubUV Animation Mode:ENiagara_SubUVLookupModeV2=::NewEnumerator1
    SubUV Curve=DataInterface() {propertyValues {Curve {keys [{arriveTangent=4.1785, time=0, value=0}, {arriveTangent=0.2086, leaveTangent=0.2086, time=0.5, value=1}, <at 1 0>]}}}
    UseEndFrame=false
    UseStartFrame=false
   ) {module=SubUVAnimation001}
   &RotateAroundPoint(
    Radius:NiagaraFloat=&RandomRangeFloat(Minimum=1, Maximum=2, Evaluation Type=::NewEnumerator1)
    Rotation Center:NiagaraPosition=$Particles.Position
    Rotation Rate:NiagaraFloat=&RandomRangeFloat(Minimum=-2, Maximum=2)
    X Axis:Vector3f=<0 1 0>
    Y Axis:Vector3f=<0 0 1>
   )
   &JitterPosition(Jitter Amount=7) {enabled=false}
   &InheritSourceMovement(
    Applied Position Scale=&VectorFromFloat(Value=&FloatFromCurve(FloatCurve=DataInterface() {propertyValues {Curve {keys [
          {interpMode=RCIM_Cubic, leaveTangent=-0.2619, tangentMode=RCTM_Break, time=0, value=1}
          {arriveTangent=-2.6487, interpMode=RCIM_Cubic, tangentMode=RCTM_Break, time=1, value=0}
         ]}}}))
    ApplyForce=false
    ApplyPosition=true
    ApplyPositionFalloff=false
   )
   &SolveForcesAndVelocity(Clamp Velocity=false, Limit Acceleration=false)
  ]
  rendererClasses [&NiagaraSpriteRendererProperties]
  renderers [{rendererClass=&NiagaraSpriteRendererProperties}]
 }
 Confetti {
  emitterUpdateScript [
   &EmitterState()
   &VC_FilteredSocketCount()
   &SpawnRate(SpawnRate=50, Use Spawn Probability=false) {enabled=false}
  ]
  particleSpawnScript [
   &VC_PaintedVariantsOverride() {enabled=false}
   &InitializeParticle(
    Color Mode=::NewEnumerator0
    Lifetime=&MakeCustomFloatFromBool(Bool=$User.bIsFrontendPreview, True Float=&RandomRangeFloat(Minimum=0.5, Maximum=0.8), False Float=0.3)
    Sprite Rotation Mode=::NewEnumerator2
    Sprite Size Mode=::NewEnumerator1
    Uniform Sprite Size Max=16
    Uniform Sprite Size Min=12
    UsePositionOffset=false
    Write Material Random=true
   )
   &PartitionParticles(Number Of Partitions=$Emitter.PortCount)
   &SkeletalMeshLocation(
    Bone / Socket Index=$Particles.PartitionParticles.Partition
    Bone / Socket Sampling Mode=::NewEnumerator5
    MaskSpawnGroup=false
    Mesh Sampling Type=::NewEnumerator6
    Skeletal Mesh=$System.SkeletalMesh
   )
   &ShapeLocation(MaskSpawnGroup=false, Override Seed=false, Sphere Radius=12, UseOwnerScale=true)
   &AddVelocity(Velocity=&MakeCustomVectorFromBool(
     Bool=$User.bIsFrontendPreview
     True Vector=&RandomRangeVector(Minimum=<-500 100 100>, Maximum=<-200 -100 -100>)
     False Vector=<0 0 0>
    ))
   &Color(Color=<rgba 1 0.2 0.05 1>)
  ]
  particleUpdateScript [
   &ParticleState()
   &ScaleSpriteSize(Scale Sprite Size Mode=::NewEnumerator0, Uniform Scale Factor=1.2)
   &AccelerationForce(Acceleration=&RandomRangeVector(Minimum=<0 0 -286>, Maximum=<0 0 -158>))
   &InheritSourceMovement(
    Applied Position Scale=&VectorFromFloat(Value=&FloatFromCurve(FloatCurve=DataInterface() {propertyValues {Curve {keys [
          {interpMode=RCIM_Cubic, tangentMode=RCTM_Break, time=0, value=1}
          {arriveTangent=-2.9025, interpMode=RCIM_Cubic, tangentMode=RCTM_Break, time=1, value=0}
         ]}}}))
    ApplyForce=false
    ApplyPosition=true
    ApplyPositionFalloff=false
   ) {enabled=false}
   &Collision(
    Advanced Aging Rate=5
    Age Colliding Particles=true
    Analytical Collision Plane Position 1:Vector3f=<0 0 20>
    Analytical Collision Plane Position 2:Vector3f=<0 0 20>
    CPU Collision Type=::NewEnumerator3
    Collision Enabled=HlslExpression() {hlslExpression="User.bIsFrontendPreview == true ? true : false"}
    Control Roll On Collision=true
    Friction=1.5
    Friction During a Bounce:NiagaraFloat=&RandomRangeFloat(Minimum=0.9, Maximum=1.2)
    KillOnCollision=false
    Module.Control Rest State Via Surface Normal=false
    Module.EnableMaxCollisionCount=false
    Module.Randomize Collision Normal Vector=false
    Restitution=&RandomRangeFloat(Minimum=0, Maximum=0.3)
    ScaleRateOfRotation=false
   ) {module=Collision001}
   &Collision(
    Advanced Aging Rate=5
    Age Colliding Particles=true
    Collision Enabled=HlslExpression() {hlslExpression="User.bIsFrontendPreview == true ? false : true"}
    Control Roll On Collision=true
    Friction=1.5
    Friction During a Bounce:NiagaraFloat=&RandomRangeFloat(Minimum=0.9, Maximum=1.2)
    KillOnCollision=false
    Module.Control Rest State Via Surface Normal=false
    Module.EnableMaxCollisionCount=false
    Module.Randomize Collision Normal Vector=false
    Restitution=&RandomRangeFloat(Minimum=0, Maximum=0.3)
    ScaleRateOfRotation=false
   ) {module=Collision002}
   &SolveForcesAndVelocity(Clamp Velocity=false, Limit Acceleration=false)
   &SubUVAnimation(
    "Sprite Renderer "=DataInterface() {propertyValues {SpriteRenderer=&self:Confetti_0.NiagaraSpriteRendererProperties_0}}
    UseEndFrame=false
    UseStartFrame=false
   )
  ]
  rendererClasses [&NiagaraSpriteRendererProperties]
  renderers [{rendererClass=&NiagaraSpriteRendererProperties}]
 }
 Left001 {
  emitterUpdateScript [
   &EmitterState(Life Cycle Mode=::NewEnumerator1, Loop Duration=1, UseLoopDelay=false)
   &VC_FilteredSocketCount()
   &VC_SpawnRibbon_v1(Number of Ribbons=$Emitter.PortCount, SpawnRate=35)
  ]
  particleSpawnScript [
   &VC_PaintedVariantsOverride() {enabled=false}
   &VC_InitializeRibbon_v1(
    Custom Offset=true
    Enable Skeletal Position=true
    Lifetime=0.7
    Local Vector Offset=&AddVectorToPosition(Position=<50 0 0>, Vector=$User.DTCarOffset)
    Offset Amount=1
    Ribbon Width=90
   )
   &AddVelocity(Velocity=<-2000 0 0>) {module=AddVelocity001}
   &Color(Color=<rgba 1.4 0.1991 0.084 1.5>) {module=Color001}
  ]
  particleUpdateScript [
   &ParticleState()
   &AccelerationForce(Acceleration=&RandomRangeVector(Minimum=<-256 -256 -1024>, Maximum=<256 256 1024>)) {module=AccelerationForce003}
   &ScaleRibbonWidth(Ribbon Width Scale=&FloatFromCurve(FloatCurve=DataInterface() {propertyValues {Curve {keys [{interpMode=RCIM_Cubic, leaveTangent=-2.2215, tangentMode=RCTM_Break, time=0, value=1.8}, <at 1 0>]}}}))
   &ScaleColor(
    Scale Alpha=&FloatFromCurve(FloatCurve=DataInterface() {propertyValues {Curve {keys [
         {arriveTangent=-0.0541, interpMode=RCIM_Cubic, time=0, value=0}
         {arriveTangent=-0.6745, interpMode=RCIM_Cubic, leaveTangent=-0.6745, tangentMode=RCTM_User, time=0.1, value=1}
         {arriveTangent=-0.0992, interpMode=RCIM_Cubic, leaveTangent=-3.8611, tangentMode=RCTM_Break, time=0.9, value=0}
        ]}}})
    Scale RGB=&VectorFromFloat(Value=&MakeFloatFromVector(
      Vector=&VectorFromFloat(Value=&FloatFromCurve(
        FloatCurve=DataInterface() {propertyValues {Curve {keys [
            {interpMode=RCIM_Cubic, leaveTangent=0.5253, tangentMode=RCTM_Break, time=0, value=0.1}
            {arriveTangent=-19.3113, interpMode=RCIM_Cubic, leaveTangent=-7.7161, tangentMode=RCTM_Break, time=0.15, value=3}
            {arriveTangent=-0.5262, interpMode=RCIM_Cubic, leaveTangent=0.5253, tangentMode=RCTM_Break, time=0.7, value=0}
           ]}}}
        Scale Curve=0.6
       ))
      Channel=::NewEnumerator0
     ))
    ScaleA=true
    ScaleRGB=true
   ) {module=ScaleColor002}
   &InheritSourceMovement(
    Applied Position Scale=&VectorFromFloat(Value=&FloatFromCurve(FloatCurve=DataInterface() {propertyValues {Curve {keys [
          {interpMode=RCIM_Cubic, tangentMode=RCTM_Break, time=0, value=1}
          {arriveTangent=-2.122, interpMode=RCIM_Cubic, tangentMode=RCTM_Break, time=1, value=0}
         ]}}}))
    ApplyForce=false
    ApplyPosition=true
    ApplyPositionFalloff=false
   )
   &SolveForcesAndVelocity(Clamp Velocity=false, Limit Acceleration=false)
   &VC_UpdateRibbon_v1(Enable Skeletal Position=true)
  ]
  rendererClasses [&NiagaraRibbonRendererProperties]
  renderers [
   {rendererClass=&NiagaraRibbonRendererProperties}
   {rendererClass=&NiagaraRibbonRendererProperties, rendererIndex=1}
  ]
 }
 notes01 {
  emitterUpdateScript [
   &EmitterState()
   &VC_FilteredSocketCount()
   &SpawnRate(
    SpawnRate=&MakeCustomFloatFromBool(Bool=$User.bIsFrontendPreview, True Float=20, False Float=25)
    Use Spawn Probability=false
   )
  ]
  particleSpawnScript [
   &VC_PaintedVariantsOverride() {enabled=false}
   &InitializeParticle(
    Color Mode=::NewEnumerator0
    Lifetime=&MakeCustomFloatFromBool(Bool=$User.bIsFrontendPreview, True Float=&RandomRangeFloat(Minimum=0.5, Maximum=1), False Float=1)
    Sprite Rotation Mode=::NewEnumerator2
    Sprite Size Max=<20 15>
    Sprite Size Min=<12 10>
    Sprite Size Mode=::NewEnumerator2
    UsePositionOffset=false
    Write Material Random=true
   )
   &PartitionParticles(Number Of Partitions=$Emitter.PortCount)
   &SkeletalMeshLocation(
    Bone / Socket Index=$Particles.PartitionParticles.Partition
    Bone / Socket Sampling Mode=::NewEnumerator5
    MaskSpawnGroup=false
    Mesh Sampling Type=::NewEnumerator6
    Skeletal Mesh=$System.SkeletalMesh
   )
   &AddVelocity(Velocity=&MakeCustomVectorFromBool(
     Bool=$User.bIsFrontendPreview
     True Vector=&RandomRangeVector(Minimum=<-500 -170 0>, Maximum=<-1000 170 300>)
     False Vector=&RandomRangeVector(Minimum=<-800 -170 0>, Maximum=<-800 170 250>)
    ))
  ]
  particleUpdateScript [
   &ParticleState()
   &ScaleSpriteSize(Scale Sprite Size Mode=::NewEnumerator0, Uniform Scale Factor=3)
   &ScaleColor(Scale RGB=&VectorFromFloat(Value=1.5), ScaleA=true, ScaleRGB=true)
   &GravityForce(Gravity=&MakeCustomVectorFromBool(
     Bool=$User.bIsFrontendPreview
     True Vector=&RandomRangeVector(Minimum=<0 0 -612>, Maximum=<0 0 -300>)
     False Vector=&RandomRangeVector(Minimum=<0 0 -600>, Maximum=<0 0 -400>)
    ))
   &Collision(
    Age Colliding Particles=false
    KillOnCollision=false
    Module.Control Rest State Via Surface Normal=false
    Module.EnableMaxCollisionCount=false
    Module.Randomize Collision Normal Vector=false
   ) {enabled=false}
   &InheritSourceMovement(
    Applied Position Scale=&VectorFromFloat(Value=&FloatFromCurve(FloatCurve=DataInterface() {propertyValues {Curve {keys [
          {interpMode=RCIM_Cubic, tangentMode=RCTM_Break, time=0, value=1}
          {arriveTangent=-2.9025, interpMode=RCIM_Cubic, tangentMode=RCTM_Break, time=1, value=0}
         ]}}}))
    ApplyForce=false
    ApplyPosition=true
    ApplyPositionFalloff=false
   )
   &Collision(
    Advanced Aging Rate=5
    Age Colliding Particles=true
    Analytical Collision Plane Position 1:Vector3f=<0 0 20>
    Analytical Collision Plane Position 2:Vector3f=<0 0 20>
    CPU Collision Type=::NewEnumerator3
    Collision Enabled=HlslExpression() {hlslExpression="User.bIsFrontendPreview == true ? true : false"}
    Control Roll On Collision=true
    Friction=1.5
    Friction During a Bounce:NiagaraFloat=&RandomRangeFloat(Minimum=0.9, Maximum=1.2)
    KillOnCollision=false
    Module.Control Rest State Via Surface Normal=false
    Module.EnableMaxCollisionCount=false
    Module.Randomize Collision Normal Vector=false
    Restitution=&RandomRangeFloat(Minimum=0, Maximum=0.3)
    ScaleRateOfRotation=false
   ) {module=Collision001}
   &Collision(
    Advanced Aging Rate=5
    Age Colliding Particles=true
    Collision Enabled=HlslExpression() {hlslExpression="User.bIsFrontendPreview == true ? false : true"}
    Control Roll On Collision=true
    Friction=1.5
    Friction During a Bounce:NiagaraFloat=&RandomRangeFloat(Minimum=0.9, Maximum=1.2)
    KillOnCollision=false
    Module.Control Rest State Via Surface Normal=false
    Module.EnableMaxCollisionCount=false
    Module.Randomize Collision Normal Vector=false
    Restitution=&RandomRangeFloat(Minimum=0, Maximum=0.3)
    ScaleRateOfRotation=false
   ) {module=Collision002}
   &SubUVAnimation(
    "Sprite Renderer "=DataInterface() {propertyValues {SpriteRenderer=&self:Bats_0.NiagaraSpriteRendererProperties_0}}
    UseEndFrame=false
    UseStartFrame=false
   )
   &SolveForcesAndVelocity(Clamp Velocity=false, Limit Acceleration=false)
   &GenerateDeathEvent(Use Event Probability=false)
  ]
  rendererClasses [&NiagaraSpriteRendererProperties]
  renderers [{rendererClass=&NiagaraSpriteRendererProperties}]
 }
}
summary {
 systemName=NS_Boost_CandyCorn
 emitters [
  {emitterName=Bats, rendererClasses [&NiagaraSpriteRendererProperties]}
  {emitterName=notes01, rendererClasses [&NiagaraSpriteRendererProperties]}
  {emitterName=Confetti, rendererClasses [&NiagaraSpriteRendererProperties]}
  {emitterName=Left001, rendererClasses [&NiagaraRibbonRendererProperties]}
 ]
 userVariables [
  User.bIsFrontendPreview {type=NiagaraBool, defaultValue=true}
  User.bIsPainted {type=NiagaraBool}
  User.ColorVariantID {type=NiagaraInt32}
  User.DTCarOffset {type=Vector3f, description="", defaultValue {struct=Vector3f, value=<0 0 0>}}
  User.PaintedColor {type=LinearColor, description="", defaultValue {struct=LinearColor, value=<rgba 1 1 1 1>}}
  User.PaintedScalar {type=NiagaraFloat, defaultValue=1, description=""}
  User.PortCount {type=NiagaraInt32, defaultValue=1}
 ]
}
